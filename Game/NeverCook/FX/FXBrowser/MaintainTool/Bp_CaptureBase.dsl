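; Bp_CaptureBase is a Blueprint in the FX browser maintenance tool, apparently a base for capture actors.
; Its construction script adds a static mesh component and sets its relative location 50 units down on Z.
; BeginPlay is empty, and the file shows no capture logic.
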(event EventBeginPlay)

(fn ConstructionScript
  (bind _returnvalue (AddComponent|NODE_AddStaticMeshComponent-0))
  (Transformation|SetRelativeLocation _returnvalue "0.000000,0.000000,-50.000000" false false))
