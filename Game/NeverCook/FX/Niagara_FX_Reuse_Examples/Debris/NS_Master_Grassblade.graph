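# Master Niagara system, filed as a reusable debris example, that emits grass-blade sprites through one GPU emitter.
# A steady spawn rate creates dark, elongated, randomly sized sprites on a torus shape, launched along +Z with several seconds of life.
# Sprites spin randomly and scale in and out by curve, while two dynamic parameter sets feed the material.
# User.RampupTimeDebris is exposed, but no shown module reads it.
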NiagaraSystem /Game/NeverCook/FX/Niagara_FX_Reuse_Examples/Debris/NS_Master_Grassblade
dependencies {usedDataInterfaces [&NiagaraDataInterfaceCurve], usedDynamicInputs [], usedModules [&SystemState]}
emitters {Debris_01 {
  simTarget=GPUComputeSim
  emitterUpdateScript [
   &EmitterState()
   &SpawnRate(SpawnRate=100, Use Spawn Probability=false)
  ]
  particleSpawnScript [
   &InitializeParticle(
    Color=<rgba 0.0104 0.0104 0.0104 5>
    Lifetime=&RandomRangeFloat(Minimum=4, Maximum=5)
    Sprite Size Max=<25 50>
    Sprite Size Min=<5 25>
    Sprite Size Mode=::NewEnumerator2
    UsePositionOffset=false
    Write Material Random=true
   )
   &TorusLocation(Align Torus With Axis=false, MaskSpawnGroup=false, Module.Override Randomness=false, Torus Mode=::NewEnumerator2)
   &AddVelocity(Velocity=<0 0 300>)
  ]
  particleUpdateScript [
   &ParticleState()
   &DynamicMaterialParameters(
    Index 0 Param 1=&FloatFromCurve(FloatCurve=DataInterface() {propertyValues {Curve {keys [<at 0 0>,<at 1 0>]}}})
    Index 0 Param 2=2
    Index 0 Param 3=&RandomRangeFloat(Minimum=500, Maximum=500)
    Index 0 Param 4=0.25
    Index 1 Param 1=512
    Module.Param0WriteEnabled=true
    Module.Param0WriteEnabled1=true
    Module.Param1WriteEnabled=true
    Module.Param1WriteEnabled1=true
    Module.Param2WriteEnabled=true
    Module.Param2WriteEnabled1=true
    Module.Param3WriteEnabled=true
    Module.Param3WriteEnabled1=true
    Write Parameter Index 1=true
   )
   &SolveForcesAndVelocity(Clamp Velocity=true, Limit Acceleration=false, Speed Limit=1250)
   &SpriteRotationRate(Rotation Rate=&RandomRangeFloat(Minimum=-300, Maximum=300))
   &ScaleSpriteSize(Uniform Curve Sprite Scale=DataInterface() {propertyValues {Curve {keys [
        {interpMode=RCIM_Cubic, tangentMode=RCTM_User, time=0, value=0}
        {interpMode=RCIM_Cubic, tangentMode=RCTM_User, time=0.1, value=1}
        {interpMode=RCIM_Cubic, tangentMode=RCTM_User, time=0.8, value=0.95}
        {tangentMode=RCTM_User, time=1, value=0}
       ]}}})
  ]
  rendererClasses [&NiagaraSpriteRendererProperties]
  renderers [{rendererClass=&NiagaraSpriteRendererProperties}]
 }}
summary {
 systemName=NS_Master_Grassblade
 emitters [{emitterName=Debris_01, simTarget=GPUComputeSim, rendererClasses [&NiagaraSpriteRendererProperties]}]
 userVariables [User.RampupTimeDebris {type=NiagaraFloat, defaultValue=1, description=""}]
}
