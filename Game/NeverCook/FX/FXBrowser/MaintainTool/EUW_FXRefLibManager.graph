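# EUW_FXRefLibManager is an editor utility widget that batch-captures cosmetics and Niagara templates for an FX reference library.
# It gathers assets per type, filters by version, existence and FX content, then equips cosmetics via CosmeticPicker widgets for PIE capture.
# Timed delays sequence commands written to a Python link file for recording and video-to-image conversion.
# Niagara templates are framed from measured bounds; finished items are logged and moved into per-type folders.
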EditorUtilityWidgetBlueprint /Game/NeverCook/FX/FXBrowser/MaintainTool/EUW_FXRefLibManager
graphs [
 CheckExist {nodes [
   n1 = CheckExist!() -> (then, Asset:Asset Data Structure) {is=title K2Node_FunctionEntry}
   > n2 = Sequence!(execute) -> (then_0, then_1, then_2:Exec, then_3:Exec, then_4:Exec, then_5:Exec, then_6:Exec, then_7:Exec)
   n3 = GetAsset(InAssetData=@n1)
   > Cast To AthenaPickaxeItemDefinition!(execute, Object=@n3) -> (then, CastFailed, AsAthena Pickaxe Item Definition:Athena Pickaxe Item Definition Object Reference) {is=title K2Node_DynamicCast}
   > Set ParentFolder(ParentFolder:String=Append(A=Get ArrangeFolderPath() -> (ArrangeFolderPath:String), B=Pickaxe/)) -> (Output_Get:String)
   Cast To AthenaBackpackItemDefinition!(execute=@n2.then_1, Object=@n3) -> (then, CastFailed, AsAthena Backpack Item Definition:Athena Backpack Item Definition Object Reference) {is=title K2Node_DynamicCast}
   > Set ParentFolder(ParentFolder=Append(A=Get ArrangeFolderPath(), B=Backpack/))
   Cast To AthenaCharacterItemDefinition!(execute=@n2.then_2, Object=@n3) -> (then, CastFailed, AsAthena Character Item Definition:Athena Character Item Definition Object Reference) {is=title K2Node_DynamicCast}
   > Set ParentFolder(ParentFolder=Append(A=Get ArrangeFolderPath(), B=Skin/))
   Cast To AthenaGliderItemDefinition!(execute=@n2.then_3, Object=@n3) -> (then, CastFailed, AsAthena Glider Item Definition:Athena Glider Item Definition Object Reference) {is=title K2Node_DynamicCast}
   > Set ParentFolder(ParentFolder=Append(A=Get ArrangeFolderPath(), B=Glider/))
   Cast To AthenaSkyDiveContrailItemDefinition!(execute=@n2.then_4, Object=@n3) -> (then, CastFailed, AsAthena Sky Dive Contrail Item Definition:Athena Sky Dive Contrail Item Definition Object Reference) {is=title K2Node_DynamicCast}
   > Set ParentFolder(ParentFolder=Append(A=Get ArrangeFolderPath(), B=SkyDive/))
   Cast To AthenaDanceItemDefinition!(execute=@n2.then_5, Object=@n3) -> (then, CastFailed, AsAthena Dance Item Definition:Athena Dance Item Definition Object Reference) {is=title K2Node_DynamicCast}
   > Set ParentFolder(ParentFolder=Append(A=Get ArrangeFolderPath(), B=Emote/))
   Cast To NiagaraSystem!(execute=@n2.then_6, Object=@n3) -> (then, CastFailed, AsNiagara System:Niagara System Object Reference) {is=title K2Node_DynamicCast}
   > Set ParentFolder(ParentFolder=Append(A=Get ArrangeFolderPath(), B=NSTemplate/))
   n4 = "Execute Python Command (Advanced)"!(
    execute=@n2.then_7
    PythonCommand:String=Append!(
     A="import os\r\ndef check_folder_exists(parent_folder, folder_name):\r\n    folder_path = os.path.join(parent_folder, folder_name)\r\n    return os.path.isdir(folder_path)\r\nparent_folder = '"
     B=Get ParentFolder() -> (ParentFolder:String)
     C:String="'\r\nfolder_name = '"
     D:String=GetObjectName(Object=@n3)
     E:String="'\r\nif check_folder_exists(parent_folder, folder_name):\r\n    print(\"True\")\r\nelse:\r\n    print(\"False\")"
    ) -> (ReturnValue)
    ExecutionMode:EPythonCommandExecutionMode Enum=ExecuteFile
    FileExecutionScope:EPythonFileExecutionScope Enum=Private
   ) -> (then, CommandResult:String, LogOutput:Array of Python Log Output Entry Structures, ReturnValue:bool)
   n5 = Break Python Log Output Entry!(PythonLogOutputEntry:"Python Log Output Entry Structure (by ref)"="Get (a ref)"(Array:Array of Python Log Output Entry Structures=@n4.LogOutput, Dimension 1=0) -> (Output:"Python Log Output Entry Structure (by ref)")) -> (Type:EPythonLogOutputType Enum, Output:String) {is=title K2Node_BreakStruct}
   > Return Node!(execute, Exist?:bool="integer >= integer"(A=FindSubstring(SearchIn=@n5.Output, Substring=True)))
  ]}
 DebugItemFXREs {nodes [
   DebugItemFXREs() {is=title K2Node_FunctionEntry}
   > n1 = GetItemFXResources!(
    execute
    self
    ItemAssetData:Asset Data Structure=CreateAssetData(InAsset=Get DebugRefAsset() -> (DebugRefAsset:Fort Item Definition Object Reference), bAllowBlueprintClass=false)
   ) -> (then, HasValidFX:bool, FXList:Array of Strings, MatList:Array of Strings, TexList:Array of Strings)
   > n2 = Branch(Condition=@n1)
   > PrintString(InString=FXRL:FX---------------------)
   > n3 = For Each Loop(Array=@n1.FXList)
   PrintString(execute=@n3.Completed, InString=FXRL:Mat---------------------)
   > n4 = For Each Loop(Array=@n1.MatList)
   PrintString(execute=@n4.Completed, InString=FXRL:Tex---------------------)
   > n5 = For Each Loop(Array=@n1.TexList)
   > PrintString(InString=@n5)
   PrintString(execute=@n4.LoopBody, InString=@n4)
   PrintString(execute=@n3.LoopBody, InString=@n3)
   PrintString(execute=@n2.else, InString=FXRL:NoFX)
  ]}
 EquipCosmetic {nodes [
   n1 = EquipCosmetic!() -> (then, Cosmetic:Fort Item Definition Object Reference) {is=title K2Node_FunctionEntry}
   > n2 = GetAllWidgetsOfClass!(
    execute
    WorldContextObject:Object Reference=Self-Reference()
    WidgetClass:User Widget Class Reference=/Game/NeverCook/ArtTools/Tech_Art/CosmeticPicker/EUW_CosmeticPickerv2.EUW_CosmeticPickerv2_C
    TopLevelOnly:bool=false
   ) -> (then, FoundWidgets:Array of EUW Cosmetic Pickerv 2 Object References)
   > n3 = GetAllWidgetsOfClass(
    WorldContextObject=Self-Reference()
    WidgetClass=/Game/NeverCook/ArtTools/Tech_Art/CosmeticPicker/EUW_CosmeticPickerv2_Advanced.EUW_CosmeticPickerv2_Advanced_C
    TopLevelOnly=false
   ) -> (FoundWidgets:Array of EUW Cosmetic Pickerv 2 Advanced Object References)
   > n4 = Sequence!(execute) -> (then_0, then_1, then_2, then_3, then_4, then_5)
   > n5 = Cast To AthenaPickaxeItemDefinition(Object=@n1)
   n6 = "Get (a copy)"(Array:Array of EUW Cosmetic Pickerv 2 Object References=@n2) -> (Output:EUW Cosmetic Pickerv 2 Object Reference)
   n7 = Get EUW_CosmeticPicker_ID_Pickaxe!(self:EUW Cosmetic Pickerv 2 Object Reference=@n6) -> (EUW_CosmeticPicker_ID_Pickaxe:EUW Cosmetic Pickerv 2 ID Object Reference)
   > Set pickaxe!(execute, pickaxe:Athena Pickaxe Item Definition Object Reference=@n5, self:EUW Cosmetic Pickerv 2 ID Object Reference=@n7) -> (then, Output_Get:Athena Pickaxe Item Definition Object Reference)
   > IDChanged!(execute, self:EUW Cosmetic Pickerv 2 ID Object Reference=@n7, PropertyName:Name=Pickaxe) -> (then)
   n8 = Cast To AthenaBackpackItemDefinition(execute=@n4.then_1, Object=@n1)
   n9 = Get EUW_CosmeticPicker_ID_Backpack!(self:EUW Cosmetic Pickerv 2 Object Reference=@n6) -> (EUW_CosmeticPicker_ID_Backpack:EUW Cosmetic Pickerv 2 ID Object Reference)
   > Set backpack!(execute, backpack:Athena Backpack Item Definition Object Reference=@n8, self:EUW Cosmetic Pickerv 2 ID Object Reference=@n9) -> (then, Output_Get:Athena Backpack Item Definition Object Reference)
   > IDChanged(self=@n9, PropertyName=Backpack)
   n10 = Cast To AthenaCharacterItemDefinition(execute=@n4.then_2, Object=@n1)
   n11 = Get EUW_CosmeticPicker_ID_Character!(self:EUW Cosmetic Pickerv 2 Object Reference=@n6) -> (EUW_CosmeticPicker_ID_Character:EUW Cosmetic Pickerv 2 ID Object Reference)
   > Set character!(execute, character:Athena Character Item Definition Object Reference=@n10, self:EUW Cosmetic Pickerv 2 ID Object Reference=@n11) -> (then, Output_Get:Athena Character Item Definition Object Reference)
   > IDChanged(self=@n11, PropertyName=Character)
   n12 = Cast To AthenaGliderItemDefinition(execute=@n4.then_3, Object=@n1)
   n13 = Get EUW_CosmeticPicker_ID_Glider!(self:EUW Cosmetic Pickerv 2 Object Reference=@n6) -> (EUW_CosmeticPicker_ID_Glider:EUW Cosmetic Pickerv 2 ID Object Reference)
   > Set glider!(execute, glider:Athena Glider Item Definition Object Reference=@n12, self:EUW Cosmetic Pickerv 2 ID Object Reference=@n13) -> (then, Output_Get:Athena Glider Item Definition Object Reference)
   > IDChanged(self=@n13, PropertyName=Glider)
   n14 = Cast To AthenaSkyDiveContrailItemDefinition(execute=@n4.then_4, Object=@n1)
   n15 = "Get (a copy)"(Array:Array of EUW Cosmetic Pickerv 2 Advanced Object References=@n3) -> (Output:EUW Cosmetic Pickerv 2 Advanced Object Reference)
   n16 = Get EUW_CosmeticPicker_ID_Contrail!(self:EUW Cosmetic Pickerv 2 Advanced Object Reference=@n15) -> (EUW_CosmeticPicker_ID_Contrail:EUW Cosmetic Pickerv 2 ID Object Reference)
   > Set contrail!(execute, contrail:Athena Sky Dive Contrail Item Definition Object Reference=@n14, self:EUW Cosmetic Pickerv 2 ID Object Reference=@n16) -> (then, Output_Get:Athena Sky Dive Contrail Item Definition Object Reference)
   > IDChanged(self=@n16, PropertyName=Contrail)
   n17 = Cast To AthenaDanceItemDefinition(execute=@n4.then_5, Object=@n1)
   > SetEmoteList!(
    execute
    self:EUW Cosmetic Pickerv 2 Emotes Object Reference=Get EUW_CosmeticPickerv2_Emotes!(self:EUW Cosmetic Pickerv 2 Advanced Object Reference=@n15) -> (EUW_CosmeticPickerv2_Emotes:EUW Cosmetic Pickerv 2 Emotes Object Reference)
    emotes:Array of Athena Dance Item Definition Object References=Make Array("[0]":Athena Dance Item Definition Object Reference=@n17) -> (Array:Array of Athena Dance Item Definition Object References)
   ) -> (then)
  ]}
 EventGraph {nodes [
   WriteStringIntoFile!(
    execute
    WriteString:String=VideoToPic
    FilePath:String=Get PythonLinkFilePath() -> (PythonLinkFilePath:String)
    Unique:bool=false
    Clear:bool=true
    __WorldContext:Object Reference
   ) -> (then)
   > n1 = AsyncEditorDelay!(execute:Exec, Seconds:float=10, MinimumFrames:int=30) -> (then:Exec, Complete:Exec) {is=title K2Node_AsyncAction}
   FinishedCaptureProgress(execute=[@n1.Complete, @n53.Complete])
   n2 = WriteStringIntoFile(WriteString=VideoToPic, FilePath=Get PythonLinkFilePath(), Unique=false, Clear=true)
   n3 = AsyncEditorDelay(execute=[@n119.Complete, @n2.then], Seconds=2, MinimumFrames=30)
   FinishedCaptureProgress(execute=@n3.Complete)
   n4 = WriteStringIntoFile(WriteString=VideoToPic, FilePath=Get PythonLinkFilePath(), Unique=false, Clear=true)
   n5 = AsyncEditorDelay(execute=[@n69.Complete, @n4.then], Seconds=2, MinimumFrames=30)
   FinishedCaptureProgress(execute=@n5.Complete)
   n6 = GetPathNameForLoadedAsset(LoadedAsset=Get CurCaptureNiagara() -> (CurCaptureNiagara:Niagara System Object Reference))
   > PrintString(InString=Split(SourceString=Split(SourceString=@n6, InStr=/, SearchDir=FromEnd), InStr=/, SearchDir=FromEnd).RightS)
   n7 = Get PythonLinkFilePath()
   WriteStringIntoFile(WriteString=VideoToPic, FilePath=@n7, Unique=false, Clear=true)
   > n8 = AsyncEditorDelay(Seconds=10, MinimumFrames=30)
   n9 = GetPathNameForLoadedAsset(execute=[@n104.Complete, @n8.Complete], LoadedAsset=Get CurCaptureNiagara())
   n10 = Split(SourceString=@n9, InStr=/, SearchDir=FromEnd)
   > WriteStringIntoFile(
    WriteString=Append!(A="Custom%&&%", B=Split(SourceString=@n10, InStr=/, SearchDir=FromEnd).RightS) -> (ReturnValue)
    FilePath=@n7
    Unique=false
    Clear=true
   )
   > n11 = AsyncEditorDelay(Seconds=1, MinimumFrames=30)
   FinishedCaptureProgress(execute=@n11.Complete)
   WriteStringIntoFile(WriteString=VideoToPic, FilePath=Get PythonLinkFilePath(), Unique=false, Clear=true)
   > n12 = AsyncEditorDelay(Seconds=10, MinimumFrames=30)
   FinishedCaptureProgress(execute=[@n76.Complete, @n12.Complete])
   WriteStringIntoFile(WriteString=VideoToPic, FilePath=Get PythonLinkFilePath(), Unique=false, Clear=true)
   > n13 = AsyncEditorDelay(Seconds=10, MinimumFrames=30)
   FinishedCaptureProgress(execute=[@n13.Complete, @n109.Complete])
   "On Clicked (Start)"() {is=title K2Node_ComponentBoundEvent}
   n14 = Get CaptureType() -> (CaptureType:CosmeticPicker_CosmeticTypes Enum=NewEnumerator0)
   n15 = Get CaptureType() -> (CaptureType=NewEnumerator0)
   > Branch(Condition=AND Boolean(
     A="Not Equal (Enum)"!(A:CosmeticPicker_CosmeticTypes Enum=@n15, B:CosmeticPicker_CosmeticTypes Enum=NewEnumerator5) -> (ReturnValue:bool)
     B="Not Equal (Enum)"(A=@n14, B=NewEnumerator6)
    ))
   > Set Count(Count:int=Get DefaultStartIndex() -> (DefaultStartIndex:int=0)) -> (Output_Get:int=0)
   > WriteStringIntoFile(WriteString=CaptureStart, FilePath=Append(A=Get ArrangeFolderPath(), B=CaptureLog.txt), Unique=false, Clear=true)
   > WriteStringIntoFile(WriteString=PrepareWindows, FilePath=Get PythonLinkFilePath(), Unique=false, Clear=true)
   > n16 = AsyncEditorDelay(Seconds=10, MinimumFrames=30)
   n17 = AsyncEditorDelay(execute=[@n16.Complete, @n27."  "], Seconds=0.3, MinimumFrames=30)
   n18 = Select(
    Option 0:Array of Asset Data Structures=Select!(
     NewEnumerator0:Array of Asset Data Structures=Get SkinList() -> (SkinList:Array of Asset Data Structures)
     NewEnumerator1:Array of Asset Data Structures=Get BackpackList() -> (BackpackList:Array of Asset Data Structures)
     NewEnumerator2:Array of Asset Data Structures=Get PickaxeList() -> (PickaxeList:Array of Asset Data Structures)
     NewEnumerator3:Array of Asset Data Structures=Get GliderList() -> (GliderList:Array of Asset Data Structures)
     NewEnumerator4:Array of Asset Data Structures=Get EmoteList() -> (EmoteList:Array of Asset Data Structures)
     NewEnumerator5:Array of Asset Data Structures
     NewEnumerator6:Array of Asset Data Structures
     NewEnumerator7:Array of Asset Data Structures=Get SkydiveList() -> (SkydiveList:Array of Asset Data Structures)
     NewEnumerator8:Array of Asset Data Structures
     NewEnumerator9:Array of Asset Data Structures
     NewEnumerator10:Array of Asset Data Structures
     NewEnumerator11:Array of Asset Data Structures
     NewEnumerator12:Array of Asset Data Structures
     NewEnumerator13:Array of Asset Data Structures
     NewEnumerator14:Array of Asset Data Structures
     NewEnumerator15:Array of Asset Data Structures
     Index:CosmeticPicker_CosmeticTypes Enum=Get CaptureType() -> (CaptureType=NewEnumerator0)
    ) -> (ReturnValue)
    Option 1:Array of Asset Data Structures=Get CaptureNSDataList() -> (CaptureNSDataList:Array of Asset Data Structures)
    Index=Get CaptureNSTemplate() -> (CaptureNSTemplate:bool)
   ) -> (ReturnValue:Array of Asset Data Structures)
   n19 = "Get (a copy)"(Array:Array of Asset Data Structures=@n18, Dimension 1=Get Count() -> (Count:int=0)) -> (Output:Asset Data Structure)
   n20 = GetAsset(InAssetData=@n19)
   n21 = GetObjectName(Object=@n20)
   n22 = Branch(execute=@n17.Complete, Condition="integer >= integer"(A=FindSubstring(SearchIn=@n21, Substring=Bean_, bUseCase=true)))
   n23 = Append(A=Get ArrangeFolderPath(), B=CaptureLog.txt)
   n24 = GetObjectName(Object=@n20)
   n25 = WriteStringIntoFile(execute=[@n22.then, @n38.else], WriteString=Append(A="NoFX : ", B=@n24), FilePath=@n23, Unique=false, Clear=false)
   n26 = Branch(
    execute=[@n83.Complete, @n45.then, @n25.then, @n48.then]
    Condition="integer >= integer"(A=Get Count() -> (Count=0), B=Length(TargetArray:Array of Asset Data Structures=@n18))
   )
   n27 = Increment Int(" "=@n26.else, Value=Get Count() -> (Count=0))
   n28 = Branch(execute=@n22.else, Condition=Get CaptureNSTemplate())
   n29 = CheckExist!(execute=[@n28.then, @n47.then], self, Asset:Asset Data Structure=@n19) -> (then, Exist?:bool) {is=title K2Node_CallFunction}
   > n30 = Branch(Condition=@n29)
   > n31 = Branch(Condition=Get OverrideExist() -> (OverrideExist:bool))
   > n32 = RemoveExist!(execute, self, Asset:Asset Data Structure=@n19) -> (then)
   n33 = Branch(execute=[@n32.then, @n30.else], Condition=Get CaptureNSTemplate())
   > n34 = Cast To NiagaraSystem(Object=@n20)
   > Set CurCaptureNiagara(CurCaptureNiagara:Niagara System Object Reference=@n34) -> (Output_Get:Niagara System Object Reference)
   > n35 = AsyncEditorDelay(Seconds=1, MinimumFrames=30)
   DoaTest(execute=@n35.Complete)
   n36 = Branch(execute=@n33.else, Condition=Get SkipNoFX() -> (SkipNoFX:bool))
   > n37 = GetItemFXResources(ItemAssetData=@n19)
   > n38 = Branch(Condition=@n37)
   n39 = Cast To FortItemDefinition!(execute=[@n36.else, @n38.then], Object=@n20) -> (then, CastFailed, AsFort Item Definition:Fort Item Definition Object Reference) {is=title K2Node_DynamicCast}
   > EquipCosmetic!(execute, self, Cosmetic:Fort Item Definition Object Reference=@n39) -> (then) {is=title K2Node_CallFunction}
   > n40 = AsyncEditorDelay(Seconds=1, MinimumFrames=30)
   WriteStringIntoFile(execute=@n40.Complete, WriteString=PIE, FilePath=Get PythonLinkFilePath(), Unique=false, Clear=true)
   > n41 = AsyncEditorDelay(Seconds=17, MinimumFrames=30)
   n42 = Get CaptureBase() -> (CaptureBase:Bp Capture Base Object Reference)
   SetupCaptureEnviormentInPIE!(
    execute=@n41.Complete
    self
    PlayerPos:Vector=Get World Location!(self:Scene Component Object Reference=Get BakePos!(self:Bp Capture Base Object Reference=@n42) -> (BakePos:Arrow Component Object Reference)) -> (ReturnValue:Vector) {is=title K2Node_CallFunction}
   ) -> (then, PIEWorld:Object Reference)
   > n43 = AsyncEditorDelay(Seconds=2, MinimumFrames=30)
   WriteStringIntoFile(execute=@n43.Complete, WriteString=GetObjectName(Object=@n20), FilePath=D:/debg.txt, Unique=false, Clear=true)
   > n44 = Switch on CosmeticPicker_CosmeticTypes!(execute, Selection:CosmeticPicker_CosmeticTypes Enum=Get CaptureType() -> (CaptureType=NewEnumerator0)) -> (
    NewEnumerator0:Exec
    NewEnumerator1:Exec
    NewEnumerator2:Exec
    NewEnumerator3:Exec
    NewEnumerator4:Exec
    NewEnumerator5:Exec
    NewEnumerator6:Exec
    NewEnumerator7:Exec
    NewEnumerator8:Exec
    NewEnumerator9:Exec
    NewEnumerator10:Exec
    NewEnumerator11:Exec
    NewEnumerator12:Exec
    NewEnumerator13:Exec
    NewEnumerator14:Exec
    NewEnumerator15:Exec
   ) {is=title K2Node_SwitchEnum}
   > SkinCaptureProcess()
   BackPackCaptureProcess(execute=@n44.NewEnumerator1)
   PickaxeCaptureProcess(execute=@n44.NewEnumerator2)
   GliderCaptureProcess(execute=@n44.NewEnumerator3)
   EmoteCaptureProcess(execute=@n44.NewEnumerator4)
   SkyDiveCaptureProcess(execute=@n44.NewEnumerator7)
   n45 = WriteStringIntoFile(execute=@n31.else, WriteString=Append(A="Exist : ", B=@n24), FilePath=@n23, Unique=false, Clear=false)
   n46 = VersionCheck!(execute=@n28.else, self, InObj:Object Reference=@n20) -> (then, VersionOK:bool)
   > n47 = Branch(Condition=@n46)
   n48 = WriteStringIntoFile(execute=@n47.else, WriteString=Append(A="VersionCheckFailed : ", B=@n24), FilePath=@n23, Unique=false, Clear=false)
   "On Clicked (GetAll)"() {is=title K2Node_ComponentBoundEvent}
   > n49 = Branch(Condition=Get CaptureNSTemplate())
   > GetAllNSTemplate()
   GetAllCosmeticAssets(execute=@n49.else)
   > n50 = GetAllActorsOfClass!(
    execute
    WorldContextObject:Object Reference
    ActorClass:Actor Class Reference=&/Bp_CaptureBase.Bp_CaptureBase_C
   ) -> (then, OutActors:Array of Bp Capture Base Object References)
   > Set CaptureBase(CaptureBase:Bp Capture Base Object Reference="Get (a copy)"(Array:Array of Bp Capture Base Object References=@n50) -> (Output:Bp Capture Base Object Reference)) -> (Output_Get:Bp Capture Base Object Reference)
   "On Clicked (Start_1)"() {is=title K2Node_ComponentBoundEvent}
   > VersionCheck(InObj=Get TestID() -> (TestID:Fort Item Definition Object Reference))
   SkyDiveCaptureProcess() {is=title K2Node_CustomEvent}
   > n51 = Branch(Condition=Get SkipWarmupForDebug() -> (SkipWarmupForDebug:bool))
   n52 = Get CaptureBase()
   SetupCaptureEnviormentInPIE(
    execute=[@n58.Complete, @n51.then]
    PlayerPos=Get World Location(self=Get FlyPos!(self:Bp Capture Base Object Reference=@n52) -> (FlyPos:Arrow Component Object Reference))
   )
   > WriteStringIntoFile(WriteString=Start, FilePath=Get PythonLinkFilePath(), Unique=false, Clear=true)
   > n53 = AsyncEditorDelay(Seconds=12, MinimumFrames=30)
   n54 = Get CaptureBase()
   n55 = SetupCaptureEnviormentInPIE(execute=@n51.else, PlayerPos=Get World Location(self=Get FlyPos(self=@n54)))
   > ExecuteConsoleCommand(WorldContextObject=@n55, Command=SetAlwaysShowContrails 1)
   > n56 = AsyncEditorDelay(Seconds=2, MinimumFrames=30)
   ExecuteConsoleCommand(execute=@n56.Complete, Command=SetAlwaysShowContrails 0)
   > n57 = AsyncEditorDelay(Seconds=2, MinimumFrames=30)
   ExecuteConsoleCommand(execute=@n57.Complete, Command=SetAlwaysShowContrails 1)
   > WriteStringIntoFile(WriteString=WarmUp, FilePath=Get PythonLinkFilePath(), Unique=false, Clear=true)
   > n58 = AsyncEditorDelay(Seconds=40, MinimumFrames=30)
   n59 = UpdateMaxBoundSize!() -> (OutputDelegate, then, DeltaTime:double=~) {is=title K2Node_CustomEvent}
   n60 = Get ReusableNiagaraNS() -> (ReusableNiagaraNS:Niagara Particle System Actor Object Reference)
   > n61 = Is Valid(InputObject=@n60)
   Set ReusableNiagaraNSFunctionTime(execute=[@n62.CastFailed, @n61.Is Not Valid], ReusableNiagaraNSFunctionTime:double=999) -> (Output_Get:double)
   > PrintString(InString=ReusableNiagaraInactive)
   n62 = Cast To FXSystemComponent!(
    execute=@n61.Is Valid
    Object=Get NiagaraComponent!(self:Niagara Particle System Actor Object Reference=@n60) -> (NiagaraComponent:Niagara Particle System Component Object Reference)
   ) -> (then, CastFailed, AsFXSystem Component:FXSystem Component Object Reference) {is=title K2Node_DynamicCast}
   > Set ReusableNiagaraNSFunctionTime(ReusableNiagaraNSFunctionTime=float + float(A=Get ReusableNiagaraNSFunctionTime() -> (ReusableNiagaraNSFunctionTime:double), B=@n59.DeltaTime))
   n63 = GetLocalBounds!(Component:Scene Component Object Reference=@n62) -> (ReturnValue:Box Sphere Bounds Structure)
   n64 = BreakBoxSphereBounds!(InBoxSphereBounds:"Box Sphere Bounds Structure (by ref)"=@n63) -> (Origin:Vector, BoxExtent:Vector, SphereRadius:float)
   n65 = BreakBoxSphereBounds(InBoxSphereBounds=Get MaxBound() -> (MaxBound:Box Sphere Bounds Structure))
   > Branch(Condition="float > float"(A=Vector Length(A=@n64.BoxExtent), B=Vector Length(A=@n65.BoxExtent)))
   > Set MaxBound(MaxBound:Box Sphere Bounds Structure=@n63) -> (Output_Get:Box Sphere Bounds Structure)
   GliderCaptureProcess() {is=title K2Node_CustomEvent}
   > n66 = Branch(Condition=Get SkipWarmupForDebug())
   n67 = Get CaptureBase()
   SetupCaptureEnviormentInPIE(
    execute=[@n72.Complete, @n66.then]
    PlayerPos=Get World Location(self=Get PlayerPos!(self:Bp Capture Base Object Reference=@n67) -> (PlayerPos:Arrow Component Object Reference))
   )
   > n68 = AsyncEditorDelay(Seconds=2, MinimumFrames=30)
   SetupCaptureEnviormentInPIE(execute=@n68.Complete, PlayerPos=Get World Location(self=Get FlyPos(self=@n67)))
   > WriteStringIntoFile(WriteString=Start, FilePath=Get PythonLinkFilePath(), Unique=false, Clear=true)
   > n69 = AsyncEditorDelay(Seconds=15, MinimumFrames=30)
   n70 = Get CaptureBase()
   SetupCaptureEnviormentInPIE(execute=@n66.else, PlayerPos=Get World Location(self=Get FlyPos(self=@n70)))
   > n71 = AsyncEditorDelay(Seconds=2, MinimumFrames=30)
   WriteStringIntoFile(execute=@n71.Complete, WriteString=WarmUp, FilePath=Get PythonLinkFilePath(), Unique=false, Clear=true)
   > n72 = AsyncEditorDelay(Seconds=60, MinimumFrames=30)
   EmoteCaptureProcess() {is=title K2Node_CustomEvent}
   > n73 = Branch(Condition=Get SkipWarmupForDebug())
   n74 = Get CaptureBase()
   SetupCaptureEnviormentInPIE(execute=[@n77.Complete, @n73.then], PlayerPos=Get World Location(self=Get PlayerPos(self=@n74)))
   > WriteStringIntoFile(WriteString=Start, FilePath=Get PythonLinkFilePath(), Unique=false, Clear=true)
   > WriteStringIntoFile(WriteString=" ", FilePath=D:/EmoteState.txt, Unique=false, Clear=true)
   > Set EmoteEndLogic(EmoteEndLogic:bool=false) -> (Output_Get:bool)
   > Set CheckEmote(CheckEmote:bool=true) -> (Output_Get:bool)
   > n75 = AsyncEditorDelay(Seconds=20, MinimumFrames=30)
   Set CheckEmote(execute=@n75.Complete, CheckEmote=false)
   > n76 = AsyncEditorDelay(Seconds=2, MinimumFrames=30)
   WriteStringIntoFile(execute=@n73.else, WriteString=WarmUp, FilePath=Get PythonLinkFilePath(), Unique=false, Clear=true)
   > n77 = AsyncEditorDelay(Seconds=60, MinimumFrames=30)
   FinishedCaptureProgress() {is=title K2Node_CustomEvent}
   n78 = "Get (a copy)"(Array:Array of Asset Data Structures=@n18, Dimension 1=Get Count() -> (Count=0)) -> (Output:Asset Data Structure)
   > WriteCosmeticInfoToOriginalFolder!(execute, self, AssetData:Asset Data Structure=@n78) -> (then)
   > n79 = AsyncEditorDelay(Seconds=8, MinimumFrames=30)
   MoveAllAssetToArrangeFolder!(execute=@n79.Complete, self, AssetData:Asset Data Structure=@n78) -> (then)
   > n80 = AsyncEditorDelay(Seconds=8, MinimumFrames=30)
   n81 = Get Count() -> (Count=0)
   n82 = GetAsset(InAssetData=@n78)
   WriteStringIntoFile(
    execute=@n80.Complete
    WriteString=Append!(
     A="CaptureFinished : "
     B=GetObjectName(Object=@n82)
     C:String="                               Index:"
     D:String="To String (Integer)"(InInt=@n81)
    ) -> (ReturnValue)
    FilePath=Append!(A=Get ArrangeFolderPath(), B=CaptureLog.txt) -> (ReturnValue)
    Unique=false
    Clear=false
   )
   > n83 = AsyncEditorDelay(Seconds=1, MinimumFrames=30)
   PickaxeCaptureProcess() {is=title K2Node_CustomEvent}
   > n84 = Branch(Condition=Get SkipWarmupForDebug())
   n85 = Get World Location(self=Get PlayerPos(self=Get CaptureBase()))
   SetupCaptureEnviormentInPIE(execute=[@n88.Complete, @n84.then], PlayerPos=@n85)
   > WriteStringIntoFile(WriteString=Start, FilePath=Get PythonLinkFilePath(), Unique=false, Clear=true)
   > n86 = AsyncEditorDelay(Seconds=40, MinimumFrames=30)
   WriteStringIntoFile(execute=@n86.Complete, WriteString=VideoToPic, FilePath=Get PythonLinkFilePath(), Unique=false, Clear=true)
   > n87 = AsyncEditorDelay(Seconds=15, MinimumFrames=30)
   FinishedCaptureProgress(execute=@n87.Complete) {is=title K2Node_CallFunction}
   WriteStringIntoFile(execute=@n84.else, WriteString=WarmUp, FilePath=Get PythonLinkFilePath(), Unique=false, Clear=true)
   > n88 = AsyncEditorDelay(Seconds=40, MinimumFrames=30)
   ResetMaxBoundSize() {is=title K2Node_CustomEvent}
   > Set MaxBound(MaxBound=MakeBoxSphereBounds!(Origin:Vector="0, 0, 0", BoxExtent:Vector="0, 0, 0", SphereRadius:float=0) -> (ReturnValue:Box Sphere Bounds Structure))
   > Set ReusableNiagaraNSFunctionTime(ReusableNiagaraNSFunctionTime=0)
   > Set ResetReusableCount(ResetReusableCount:int=0) -> (Output_Get:int=0)
   n89 = ResetReusableNiagaraNS!() -> (OutputDelegate, then, Delta:double=~) {is=title K2Node_CustomEvent}
   > n90 = Set ReusableNiagaraNSFunctionTime(ReusableNiagaraNSFunctionTime=float + float(A=@n89.Delta, B=Get ReusableNiagaraNSFunctionTime()))
   > Branch(Condition="float >= float"(A=@n90, B=3))
   n91 = Get ReusableNiagaraNS()
   n92 = Get NiagaraComponent(self=@n91)
   > SetActive!(execute, self:Actor Component Object Reference=@n92, bNewActive:bool=false, bReset:bool=true) -> (then)
   > Deactivate(self:Actor Component Object Reference=@n92)
   n93 = Get NiagaraComponent(self=@n91)
   > Set Niagara System Asset!(
    execute
    self:Niagara Particle System Component Object Reference=@n93
    InAsset:Niagara System Object Reference
    bResetExistingOverrideParameters:bool=true
   ) -> (then) {is=title K2Node_CallFunction}
   > SetActive(self=@n92, bNewActive=false, bReset=true)
   > Deactivate(self=@n92)
   > Set Niagara System Asset(self=@n93, InAsset=Get CurCaptureNiagara(), bResetExistingOverrideParameters=true)
   > Reset System(self:Niagara Particle System Component Object Reference=@n93)
   > Set ReusableNiagaraNSFunctionTime(ReusableNiagaraNSFunctionTime=0)
   > n94 = Increment Int(Value=Get ResetReusableCount() -> (ResetReusableCount:int=0))
   > Branch(Condition="integer >= integer"(A=@n94, B=2))
   > Set StartReusableNiagaraNSReset(StartReusableNiagaraNSReset:bool=false) -> (Output_Get:bool)
   > Destroy Actor(self=Get ReusableNiagaraNS())
   DoaTest() {is=title K2Node_CustomEvent}
   > SetLevelViewportCameraInfo!(
    execute
    self:Unreal Editor Subsystem Object Reference=Get UnrealEditorSubsystem()
    CameraLocation:Vector="0, 0, 0"
    CameraRotation:Rotator="0,45,0"
   ) -> (then)
   n95 = Sqrt!(A:double=float / float(A=float * float(A=500, B=500) {is=title K2Node_PromotableOperator}, B=2)) -> (ReturnValue:double)
   > n96 = SpawnActorFromClass!(
    execute
    ActorClass:Actor Class Reference=/Script/Niagara.NiagaraActor
    Location:Vector=MakeVector(X=@n95, Y=@n95)
    Rotation:Rotator
    bTransient:bool
   ) -> (then, ReturnValue:Niagara Particle System Actor Object Reference)
   > Set ReusableNiagaraNS(ReusableNiagaraNS:Niagara Particle System Actor Object Reference=@n96) -> (Output_Get:Niagara Particle System Actor Object Reference)
   n97 = Get NiagaraComponent(self=@n96)
   > Set Niagara System Asset(self=@n97, InAsset=Get CurCaptureNiagara(), bResetExistingOverrideParameters=true)
   > Cast To FXSystemComponent(Object=@n97)
   > ResetMaxBoundSize() {is=title K2Node_CallFunction}
   > Set StartReusableNiagaraNSBoundCheck(StartReusableNiagaraNSBoundCheck:bool=true) -> (Output_Get:bool)
   > Delay!(execute, WorldContextObject:Object Reference, Duration=3.5) -> (then)
   n98 = BreakBoxSphereBounds(InBoxSphereBounds=Get MaxBound())
   n99 = Get ReusableNiagaraNS()
   > DrawDebugBox!(
    execute
    WorldContextObject:Object Reference=Get ReusableNiagaraNS()
    Center:Vector=vector + vector(A=Get Actor Location(self=@n99), B=@n98)
    Extent:Vector=@n98.BoxExtent
    LineColor:Linear Color Structure=(R=0.2476,G=1,B=0.0096,A=1)
    Rotation:Rotator
    Duration:float=1
    Thickness:float=2
    DepthPriority:EDrawDebugSceneDepthPriorityGroup Enum=World
   ) -> (then)
   n100 = Sqrt(A=float / float(A=float * float(A=500, B=500), B=2))
   n101 = "Absolute (Float)"(A=BreakVector(InVec=@n98.BoxExtent))
   n102 = Sqrt(A=float / float(A=float * float(A=@n101, B=@n101), B=2))
   n103 = MakeVector(X=@n102, Y=@n102)
   > SetLevelViewportCameraInfo(
    self=Get UnrealEditorSubsystem()
    CameraLocation=vector + vector!(A=MakeVector(X=@n100, Y=@n100), B=@n98, C:Vector=vector * vector(A=@n103, B:Vector="-1,-1,-1") -> (ReturnValue=~), D:Vector) -> (ReturnValue)
    CameraRotation="0,45,0"
   )
   > Delay(Duration=1.5)
   > Set StartReusableNiagaraNSReset(StartReusableNiagaraNSReset=true)
   > Set ReusableNiagaraNSFunctionTime(ReusableNiagaraNSFunctionTime=3)
   > WriteStringIntoFile(WriteString=Start, FilePath=Get PythonLinkFilePath(), Unique=false, Clear=true)
   > n104 = AsyncEditorDelay(Seconds=5, MinimumFrames=30)
   SkinCaptureProcess() {is=title K2Node_CustomEvent}
   > n105 = Branch(Condition=Get SkipWarmupForDebug())
   SetupCaptureEnviormentInPIE(execute=[@n110.Complete, @n105.then], PlayerPos=@n85)
   > WriteStringIntoFile(WriteString=Start, FilePath=Get PythonLinkFilePath(), Unique=false, Clear=true)
   > n106 = AsyncEditorDelay(Seconds=7.2, MinimumFrames=30)
   n107 = GetPIEWorlds!(execute=@n106.Complete, bIncludeDedicatedServer:bool=false) -> (then, ReturnValue:Array of World Object References)
   > SetControlRotation!(
    execute
    self:Controller Object Reference=GetPlayerController!(
     WorldContextObject:Object Reference="Get (a copy)"(Array:Array of World Object References=@n107) -> (Output:World Object Reference)
     PlayerIndex:int=0
    ) -> (ReturnValue:Player Controller Object Reference)
    NewRotation:"Rotator (by ref)"=MakeRotator(Yaw=90)
   ) -> (then)
   > n108 = AsyncEditorDelay(Seconds=2, MinimumFrames=30)
   n109 = AsyncEditorDelay(execute=[@n108.Complete, @n133.then], Seconds=5, MinimumFrames=30)
   WriteStringIntoFile(execute=@n105.else, WriteString=WarmUp, FilePath=Get PythonLinkFilePath(), Unique=false, Clear=true)
   > n110 = AsyncEditorDelay(Seconds=40, MinimumFrames=30)
   GetAllNSTemplate() {is=title K2Node_CustomEvent}
   > n111 = GetFolderListInPath!(
    execute
    "Path(/end)":String=H:/WorkSpace/yi.xie_ForFXBrowserTest/FortniteGame/Content/NeverCook/FX/Niagara_FX_Reuse_Examples/
    __WorldContext:Object Reference
   ) -> (then, FolderList:Array of Strings, FullPathList:Array of Strings)
   > n112 = For Each Loop(Array=@n111.FullPathList)
   > n113 = GetFileListInPath!(execute, "Path(/end)":String=@n112, ext:String=uasset, __WorldContext:Object Reference) -> (then, FileList:Array of Strings, FullPathList:Array of Strings)
   > n114 = For Each Loop(Array=@n113.FullPathList)
   n115 = Append(A=/Game/, B=Split(SourceString=@n114, InStr=/Content/, SearchDir=FromStart).RightS)
   > n116 = LoadAsset(AssetPath=Replace(SourceString=Replace(SourceString=@n115, From='), From=.uasset))
   > n117 = Cast To NiagaraSystem(Object=@n116)
   > Add(
    TargetArray:Array of Niagara System Object References=Get CaptureNSList() -> (CaptureNSList:Array of Niagara System Object References)
    NewItem:"Niagara System Object Reference (by ref)"=@n117
   ) {is=title K2Node_CallArrayFunction}
   > Add(
    TargetArray:Array of Asset Data Structures=Get CaptureNSDataList()
    NewItem:"Asset Data Structure (by ref)"=CreateAssetData(InAsset=@n117, bAllowBlueprintClass=false)
   )
   BackPackCaptureProcess() {is=title K2Node_CustomEvent}
   > n118 = Branch(Condition=Get SkipWarmupForDebug())
   SetupCaptureEnviormentInPIE(execute=[@n120.Complete, @n118.then], PlayerPos=@n85)
   > WriteStringIntoFile(WriteString=Start, FilePath=Get PythonLinkFilePath(), Unique=false, Clear=true)
   > n119 = AsyncEditorDelay(Seconds=10, MinimumFrames=30)
   WriteStringIntoFile(execute=@n118.else, WriteString=WarmUp, FilePath=Get PythonLinkFilePath(), Unique=false, Clear=true)
   > n120 = AsyncEditorDelay(Seconds=40, MinimumFrames=30)
   Event Construct()
   > SetObject(self=Get DetailsView_0() -> (DetailsView_0:Details View Object Reference), NewObject=Self-Reference())
   n121 = Event Tick()
   > n122 = Sequence!(execute) -> (then_0, then_1, then_2, then_3)
   > Branch(Condition=Get RotateCharacter() -> (RotateCharacter:bool))
   > n123 = GetPIEWorlds(bIncludeDedicatedServer=false)
   n124 = GetPlayerController(WorldContextObject="Get (a copy)"(Array=@n123), PlayerIndex=0)
   n125 = BreakRotator(InRot=GetControlRotation!(self:Controller Object Reference=@n124) -> (ReturnValue:Rotator))
   n126 = MakeRotator(Roll=@n125, Pitch=@n125.Pitch, Yaw=float + float(A=@n125.Yaw, B=float * float(A=@n121.InDeltaTime, B=100)))
   > PrintString(InString="To String (Rotator)"!(InRot:Rotator=@n126) -> (ReturnValue:String))
   > SetControlRotation(self=@n124, NewRotation=@n126)
   Branch(execute=@n122.then_1, Condition=Get CheckEmote() -> (CheckEmote:bool))
   > n127 = GetPIEWorlds(bIncludeDedicatedServer=false)
   > n128 = Cast To FortPlayerPawn(Object=GetPlayerPawn!(WorldContextObject:Object Reference="Get (a copy)"(Array=@n127), PlayerIndex:int=0) -> (ReturnValue:Pawn Object Reference))
   > n129 = Branch(Condition=IsPlayingEmote!(self:Fort Player Animation Interface Object Reference=GetAnimationInterface!(self:Fort Player Pawn Object Reference=@n128) -> (ReturnValue:Fort Player Animation Interface Object Reference)) -> (ReturnValue:bool))
   n130 = Get Image_0() -> (Image_0:Image Object Reference)
   > SetColorAndOpacity(self:Image Object Reference=@n130, InColorAndOpacity:Linear Color Structure=(R=1,G=0,B=0.4141,A=1))
   > Set EmoteEndLogic(EmoteEndLogic=true)
   SetColorAndOpacity(execute=@n129.else, self=@n130, InColorAndOpacity=(R=0.2607,G=1,B=0,A=1))
   > Branch(Condition=Get EmoteEndLogic() -> (EmoteEndLogic:bool))
   > Set EmoteEndLogic(EmoteEndLogic=false)
   > WriteStringIntoFile(WriteString=EmoteEnd, FilePath=D:/EmoteState.txt, Unique=false, Clear=true)
   > PrintString(InString=WWWWWWWWWWWWWWWWWWWWWWWWWWWWWWWWWWWWWWWWWWWWWWWWWWWWWWWWWWWWWWWWWWWWWWWWWWWWWWWWWWWWWWWWWWWWWWWWWWWWWWWWWWWWWWWWWWWWWWWWWWWWWWWWWWWWWWWWWWWWWWWWWWWWWWWWWWWWWWWWWWWWWWWWWWWWWWWWWWWWWWWWWWWWWWWWWWWWWWWWWWWWWWWWWWWWWWWWWWWWWWWWWWWWWWWWWWWWWWWWWWWWWWWWWWWWWWWWWWWWWWWWWWWWWWWWWWWWWWWWWWWWWWWWWWWWWWWWWWWWWWWWWWWWWWWWWWWWWWWWWWWWWWWWWWWWWWWWWWWWWWWWWWWWWWWWWWWWWWWWWWWWWWWWWWWWWWWWWWWWWWWWWWWWWWWWWWWWWWWWWWWWWWWWWWWWWWWWWWWW)
   Branch(execute=@n122.then_2, Condition=Get StartReusableNiagaraNSBoundCheck() -> (StartReusableNiagaraNSBoundCheck:bool))
   > n131 = Branch(Condition="float >= float"(A=Get ReusableNiagaraNSFunctionTime(), B=3))
   > Set StartReusableNiagaraNSBoundCheck(StartReusableNiagaraNSBoundCheck=false)
   UpdateMaxBoundSize!(execute=@n131.else, self, DeltaTime:double=@n121.InDeltaTime) -> (then) {is=title K2Node_CallFunction}
   Branch(execute=@n122.then_3, Condition=Get StartReusableNiagaraNSReset() -> (StartReusableNiagaraNSReset:bool))
   > ResetReusableNiagaraNS!(execute, self, Delta:double=@n121.InDeltaTime) -> (then) {is=title K2Node_CallFunction}
   Set RotateCharacter(RotateCharacter:bool=true) -> (Output_Get:bool)
   > n132 = AsyncEditorDelay(Seconds=4.8, MinimumFrames=30)
   n133 = Set RotateCharacter(execute=@n132.Complete, RotateCharacter=false)
   Record Pickaxe Loop()
   LoadAsset(AssetPath=/Game/NeverCook/FX/Niagara_FX_Reuse_Examples/Debris/NS_Master_Grassblade)
   Event PreConstruct()
   Get PythonLinkFilePath()
  ]}
 GetAllCosmeticAssets {nodes [
   GetAllCosmeticAssets() {is=title K2Node_FunctionEntry}
   > PrintString(InString=FXRL:StartCheckAssets)
   > n1 = Sequence!(execute) -> (then_0, then_1, then_2, then_3, then_4:Exec, then_5:Exec, then_6:Exec)
   > Clear(TargetArray:Array of Asset Data Structures=Get PickaxeList())
   > n2 = GetAssetsByClass!(
    execute
    self:Asset Registry Interface=GetAssetRegistry()
    ClassPathName:Top Level Asset Path Structure="Get Class Path Name (Top Level Asset Path)"!(Class:Object Class Reference=/Script/FortniteGame.AthenaPickaxeItemDefinition) -> (ReturnValue:Top Level Asset Path Structure) {is=title K2Node_CallFunction}
    bSearchSubClasses:bool=false
   ) -> (then, OutAssetData:Array of Asset Data Structures, ReturnValue:bool)
   > Set PickaxeList(PickaxeList:Array of Asset Data Structures=@n2) -> (Output_Get:Array of Asset Data Structures)
   Clear(execute=@n1.then_1, TargetArray=Get SkinList())
   > n3 = GetAssetsByClass(
    self=GetAssetRegistry()
    ClassPathName="Get Class Path Name (Top Level Asset Path)"(Class=/Script/FortniteGame.AthenaCharacterItemDefinition)
    bSearchSubClasses=false
   )
   > Set SkinList(SkinList:Array of Asset Data Structures=@n3) -> (Output_Get:Array of Asset Data Structures)
   Clear(execute=@n1.then_2, TargetArray=Get EmoteList())
   > n4 = GetAssetsByClass(
    self=GetAssetRegistry()
    ClassPathName="Get Class Path Name (Top Level Asset Path)"(Class=/Script/FortniteGame.AthenaDanceItemDefinition)
    bSearchSubClasses=false
   )
   > Set EmoteList(EmoteList:Array of Asset Data Structures=@n4) -> (Output_Get:Array of Asset Data Structures)
   Clear(execute=@n1.then_3, TargetArray=Get SkydiveList())
   > n5 = GetAssetsByClass(
    self=GetAssetRegistry()
    ClassPathName="Get Class Path Name (Top Level Asset Path)"(Class=/Script/FortniteGame.AthenaSkyDiveContrailItemDefinition)
    bSearchSubClasses=false
   )
   > Set SkydiveList(SkydiveList:Array of Asset Data Structures=@n5) -> (Output_Get:Array of Asset Data Structures)
   Clear(execute=@n1.then_4, TargetArray=Get GliderList())
   > n6 = GetAssetsByClass(
    self=GetAssetRegistry()
    ClassPathName="Get Class Path Name (Top Level Asset Path)"(Class=/Script/FortniteGame.AthenaGliderItemDefinition)
    bSearchSubClasses=false
   )
   > Set GliderList(GliderList:Array of Asset Data Structures=@n6) -> (Output_Get:Array of Asset Data Structures)
   Clear(execute=@n1.then_5, TargetArray=Get BackpackList())
   > n7 = GetAssetsByClass(
    self=GetAssetRegistry()
    ClassPathName="Get Class Path Name (Top Level Asset Path)"(Class=/Script/FortniteGame.AthenaBackpackItemDefinition)
    bSearchSubClasses=false
   )
   > Set BackpackList(BackpackList:Array of Asset Data Structures=@n7) -> (Output_Get:Array of Asset Data Structures)
   PrintString(execute=@n1.then_6, InString=FXRL:FinishedCheckAssets)
  ]}
 GetDependenciesList {nodes [
   n1 = GetDependenciesList!() -> (then, AssetDataInput:Asset Data Structure, Depth:int, MaxDepth:int, DepList:Array of Names) {is=title K2Node_FunctionEntry}
   n2 = Split(
    SourceString=GetFullName!(InAssetData:"Asset Data Structure (by ref)"=@n1) -> (ReturnValue:String)
    InStr=" "
    SearchCase=CaseSensitive
    SearchDir=FromStart
   )
   > n3 = Get Dependencies!(
    execute
    self:Asset Registry Interface=GetAssetRegistry()
    PackageName:Name=String To Name(InString=Split(SourceString=@n2.RightS, InStr=., SearchDir=FromStart))
    DependencyOptions:"Asset Registry Dependency Options Structure (by ref)"=Make Asset Registry Dependency Options(bIncludeSoftPackageReferences=True)
   ) -> (then, OutDependencies:Array of Names, ReturnValue:bool) {is=title K2Node_CallFunction}
   > n4 = For Each Loop(Array:Array of Names=@n3) -> (Array Element:Name)
   > n5 = Branch(Condition="integer > integer"(A=Find Item!(TargetArray:Array of Names=@n1.DepList, ItemToFind:"Name (by ref)"=@n4) -> (ReturnValue), B=-1))
   n6 = "To String (Name)"(InName=@n4)
   n7 = FindSubstring(SearchIn=@n6, Substring=/Niagara/)
   n8 = FindSubstring(SearchIn=@n6, Substring=/Script/)
   n9 = FindSubstring(SearchIn=@n6, Substring=/Engine/)
   n10 = Branch(
    execute=@n5.else
    Condition=OR Boolean!(A="Equal (Integer)"(A=@n8), B="Equal (Integer)"(A=@n9), C:bool="Equal (Integer)"(A=@n7)) -> (ReturnValue)
   )
   n11 = FindAssetData!(execute=@n10.else, AssetPath:String=@n6) -> (then, ReturnValue:Asset Data Structure)
   > n12 = IsADependencyBanned!(execute, AssetInput:String=@n6, AssetData:Asset Data Structure=@n11, __WorldContext:Object Reference) -> (then, PassCheck:bool)
   > Branch(Condition=@n12)
   > Add Unique(TargetArray:Array of Names=@n1.DepList, NewItem:"Name (by ref)"=@n4)
   > n13 = Branch(Condition="integer > integer"(A=@n1.Depth, B=@n1.MaxDepth))
   GetDependenciesList!(
    execute=@n13.else
    self
    AssetDataInput:Asset Data Structure=@n11
    Depth:int=int + int(A=@n1.Depth)
    MaxDepth:int=@n1.MaxDepth
    DepList:Array of Names=@n1.DepList
   ) -> (then) {is=title K2Node_CallFunction}
  ]}
 GetItemFXResources {nodes [
   n1 = GetItemFXResources!() -> (then, ItemAssetData:Asset Data Structure) {is=title K2Node_FunctionEntry}
   > GetDependenciesList(AssetDataInput=@n1, Depth=0, MaxDepth=7, DepList=Get tmpDependenciesList() -> (tmpDependenciesList:Array of Names))
   > n2 = For Each Loop(Array=Get tmpDependenciesList())
   Clear(execute=@n2.Completed, TargetArray:Array of Names=Get tmpDependenciesList())
   > n3 = For Each Loop(Array:Array of Strings=Get tmpFXList() -> (tmpFXList:Array of Strings)) -> (Array Element:String)
   n4 = Get tmpFXList()
   Return Node!(
    execute=@n3.Completed
    HasValidFX:bool="integer >= integer"(A=Length(TargetArray:Array of Strings=@n4), B=1)
    FXList:Array of Strings=@n4
    MatList:Array of Strings=Get tmpMatList() -> (tmpMatList:Array of Strings)
    TexList:Array of Strings=Get tmpTexList() -> (tmpTexList:Array of Strings)
   )
   > n5 = FindAssetData(AssetPath=@n3)
   > GetDependenciesList(AssetDataInput=@n5, Depth=0, MaxDepth=3, DepList=Get tmpDependenciesList())
   > n6 = For Each Loop(Array:Array of Names=Get tmpDependenciesList()) -> (Array Element:Name)
   n7 = "To String (Name)"(InName=@n6)
   > n8 = FindAssetData(AssetPath=@n7)
   n9 = FindAssetNativeClass!(AssetData:"Asset Data Structure (by ref)"=@n8) -> (ReturnValue:Object Class Reference)
   > n10 = Cast To MaterialInterface Class!(execute:Exec, Class:Object Class Reference=@n9) -> (then:Exec, CastFailed:Exec, AsMaterial Interface:Material Interface Class Reference) {is=title K2Node_ClassDynamicCast}
   > Add Unique(TargetArray:Array of Strings=Get tmpMatList(), NewItem:"String (by ref)"=@n7)
   Cast To Texture2D Class!(execute:Exec=@n10.CastFailed, Class:Object Class Reference=@n9) -> (then:Exec, CastFailed:Exec, AsTexture 2D:Texture 2D Class Reference) {is=title K2Node_ClassDynamicCast}
   > Add Unique(TargetArray=Get tmpTexList(), NewItem=@n7)
   n11 = "To String (Name)"(InName=@n2)
   n12 = FindAssetData(execute=@n2.LoopBody, AssetPath=@n11)
   n13 = FindAssetNativeClass(AssetData=@n12)
   > n14 = Cast To NiagaraSystem Class!(execute:Exec, Class:Object Class Reference=@n13) -> (then:Exec, CastFailed:Exec, AsNiagara System:Niagara System Class Reference) {is=title K2Node_ClassDynamicCast}
   n15 = Get tmpFXList()
   > Add(TargetArray:Array of Strings=@n15, NewItem:"String (by ref)"=@n11)
   Cast To ParticleSystem Class!(execute:Exec=@n14.CastFailed, Class:Object Class Reference=@n13) -> (then:Exec, CastFailed:Exec, AsParticle System:Particle System Class Reference) {is=title K2Node_ClassDynamicCast}
   > Add(TargetArray=@n15, NewItem=@n11)
  ]}
 "IterateSearchDependency(old)" {nodes [
   n1 = "IterateSearchDependency(old)"!() -> (then, Asset:Name, IDName:String, Depth:int) {is=title K2Node_FunctionEntry}
   > n2 = Get Dependencies(self=GetAssetRegistry(), PackageName=@n1, DependencyOptions=Make Asset Registry Dependency Options(bIncludeSoftPackageReferences=True))
   > n3 = For Each Loop(Array=@n2)
   n4 = "To String (Name)"(InName=@n3)
   > n5 = Branch(Condition=OR Boolean(A=Contains(SearchIn=@n4, Substring=/Script/), B=Contains(SearchIn=@n4, Substring=/Engine/), C=Contains(SearchIn=@n4, Substring=/Niagara/)))
   n6 = FindAssetData(execute=@n5.else, AssetPath=@n4)
   > n7 = Branch(Condition="integer <= integer"(A=@n1.Depth, B=1))
   > n8 = Branch(Condition="Equal (Integer)"(A=@n1.Depth))
   n9 = FindAssetNativeClass(AssetData=@n6)
   > n10 = Branch(Condition=Cast To FortWeaponMeleeItemDefinition Class!(Class:Object Class Reference=@n9) -> (AsFort Weapon Melee Item Definition:Fort Weapon Melee Item Definition Class Reference, bSuccess:bool=~) {is=title K2Node_ClassDynamicCast}.bSuccess)
   "IterateSearchDependency(old)"!(execute=[@n14.then, @n10.then], self, Asset:Name=@n3, IDName:String=@n1.IDName, Depth:int=int + int(A=@n1.Depth)) -> (then) {is=title K2Node_CallFunction}
   n11 = Cast To NiagaraSystem Class!(Class=@n9) -> (AsNiagara System, bSuccess:bool=~)
   n12 = Cast To ParticleSystem Class!(Class=@n9) -> (AsParticle System, bSuccess:bool=~)
   n13 = Branch(
    execute=@n8.else
    Condition=OR Boolean!(
     A=Cast To B_Athena_Pickaxe_Generic Class!(Class:Object Class Reference=@n9) -> (AsB Athena Pickaxe Generic:B Athena Pickaxe Generic Class Reference, bSuccess:bool=~) {is=title K2Node_ClassDynamicCast}.bSuccess
     B=Cast To B_Athena_Pickaxe_DualWield_Generic Class!(Class:Object Class Reference=@n9) -> (AsB Athena Pickaxe Dual Wield Generic:B Athena Pickaxe Dual Wield Generic Class Reference, bSuccess:bool=~) {is=title K2Node_ClassDynamicCast}.bSuccess
     C=@n11.bSuccess
     D:bool=@n12.bSuccess
    ) -> (ReturnValue)
   )
   n14 = ExecutePythonCommand(
    execute=[@n13.then, @n15.then]
    PythonCommand=Append!(
     A="import os\r\nnew_string = \""
     B=@n4
     C:String="\\n\"\r\nfile_path = \"H:/Capture/Original/"
     D:String=@n1.IDName
     E:String=".txt\"\r\n\r\n## 检查文件是否存在\r\nif not os.path.exists(file_path):\r\n    ## 如果文件不存在，创建新文件并写入字符串\r\n    with open(file_path, 'w') as f:\r\n        f.write(new_string)\r\nelse:\r\n    ## 如果文件已存在，检查字符串是否已存在\r\n    with open(file_path, 'r') as f:\r\n        lines = f.readlines()\r\n\r\n    if new_string not in lines:\r\n        ## 如果字符串不存在，将新的一行字符串添加到文件\r\n        with open(file_path, 'a') as f:\r\n            f.write(new_string)"
    ) -> (ReturnValue)
   )
   n15 = Branch(
    execute=@n7.else
    Condition=OR Boolean(
     A=Cast To MaterialInterface Class!(Class=@n9) -> (AsMaterial Interface, bSuccess:bool=~).bSuccess
     B=Cast To Texture2D Class!(Class=@n9) -> (AsTexture 2D, bSuccess:bool=~).bSuccess
     C=@n11.bSuccess
     D=@n12.bSuccess
    )
   )
  ]}
 MoveAllAssetToArrangeFolder {nodes [
   n1 = MoveAllAssetToArrangeFolder!() -> (then, AssetData:Asset Data Structure) {is=title K2Node_FunctionEntry}
   > n2 = Sequence!(execute) -> (then_0, then_1)
   > n3 = Sequence!(execute) -> (then_0, then_1, then_2:Exec, then_3:Exec, then_4:Exec, then_5:Exec, then_6:Exec, then_7:Exec)
   n4 = GetAsset(InAssetData=@n1)
   > Cast To AthenaPickaxeItemDefinition(Object=@n4)
   > Set ParentFolder(ParentFolder=Append!(A=Get ArrangeFolderPath(), B=Pickaxe/) -> (ReturnValue))
   Cast To AthenaBackpackItemDefinition(execute=@n3.then_1, Object=@n4)
   > Set ParentFolder(ParentFolder=Append(A=Get ArrangeFolderPath(), B=Backpack/))
   Cast To AthenaCharacterItemDefinition(execute=@n3.then_2, Object=@n4)
   > Set ParentFolder(ParentFolder=Append(A=Get ArrangeFolderPath(), B=Skin/))
   Cast To AthenaGliderItemDefinition(execute=@n3.then_3, Object=@n4)
   > Set ParentFolder(ParentFolder=Append(A=Get ArrangeFolderPath(), B=Glider/))
   Cast To AthenaSkyDiveContrailItemDefinition(execute=@n3.then_4, Object=@n4)
   > Set ParentFolder(ParentFolder=Append(A=Get ArrangeFolderPath(), B=SkyDive/))
   Cast To AthenaDanceItemDefinition(execute=@n3.then_5, Object=@n4)
   > Set ParentFolder(ParentFolder=Append(A=Get ArrangeFolderPath(), B=Emote/))
   Cast To NiagaraSystem(execute=@n3.then_6, Object=@n4)
   > Set ParentFolder(ParentFolder=Append(A=Get ArrangeFolderPath(), B=NSTemplate/))
   n5 = Append(A=Get ParentFolder(), B=GetObjectName(Object=GetAsset(InAssetData=@n1)))
   ExecutePythonCommand(
    execute=@n2.then_1
    PythonCommand=Append!(
     A="import os\r\nimport shutil\r\n\r\n\r\ndef move_files(source_directory, destination_directory):\r\n    files = os.listdir(source_directory)\r\n    for file_name in files:\r\n        shutil.move(os.path.join(source_directory, file_name), destination_directory)\r\n"
     B="source_folder = r'"
     C:String=Get OriginalFolderPath() -> (OriginalFolderPath:String)
     D:String="'\r\n"
     E:String=Append!(
      A="destination_folder = r'"
      B=@n5
      C="'\r\ndef ensure_directory_exists(directory):\r\n    if not os.path.exists(directory):\r\n        os.makedirs(directory)\r\nensure_directory_exists(destination_folder)\r\n"
     ) -> (ReturnValue)
     F:String="move_files(source_folder, destination_folder)"
    ) -> (ReturnValue)
   )
  ]}
 RemoveExist {nodes [
   n1 = RemoveExist!() -> (then, Asset:Asset Data Structure) {is=title K2Node_FunctionEntry}
   > n2 = Sequence()
   n3 = GetAsset(InAssetData=@n1)
   > Cast To AthenaPickaxeItemDefinition(Object=@n3)
   > Set ParentFolder(ParentFolder=Append!(A=Get ArrangeFolderPath(), B=Pickaxe/) -> (ReturnValue))
   Cast To AthenaBackpackItemDefinition(execute=@n2.then_1, Object=@n3)
   > Set ParentFolder(ParentFolder=Append(A=Get ArrangeFolderPath(), B=Backpack/))
   Cast To AthenaCharacterItemDefinition(execute=@n2.then_2, Object=@n3)
   > Set ParentFolder(ParentFolder=Append(A=Get ArrangeFolderPath(), B=Skin/))
   Cast To AthenaGliderItemDefinition(execute=@n2.then_3, Object=@n3)
   > Set ParentFolder(ParentFolder=Append(A=Get ArrangeFolderPath(), B=Glider/))
   Cast To AthenaSkyDiveContrailItemDefinition(execute=@n2.then_4, Object=@n3)
   > Set ParentFolder(ParentFolder=Append(A=Get ArrangeFolderPath(), B=SkyDive/))
   Cast To AthenaDanceItemDefinition(execute=@n2.then_5, Object=@n3)
   > Set ParentFolder(ParentFolder=Append(A=Get ArrangeFolderPath(), B=Emote/))
   Cast To NiagaraSystem(execute=@n2.then_6, Object=@n3)
   > Set ParentFolder(ParentFolder=Append(A=Get ArrangeFolderPath(), B=NSTemplate/))
   "Execute Python Command (Advanced)"(
    execute=@n2.then_7
    PythonCommand=Append!(
     A="import os\r\nimport shutil\r\ndef check_folder_exists(parent_folder, folder_name):\r\n    folder_path = os.path.join(parent_folder, folder_name)\r\n    if os.path.isdir(folder_path):\r\n        shutil.rmtree(folder_path)\r\nparent_folder = '"
     B=Get ParentFolder()
     C:String="'\r\nfolder_name = '"
     D:String=GetObjectName(Object=@n3)
     E:String="'\r\ncheck_folder_exists(parent_folder, folder_name)"
    ) -> (ReturnValue)
    ExecutionMode=ExecuteFile
    FileExecutionScope=Private
   )
   > Return Node!(execute)
  ]}
 SetupCaptureEnviormentInPIE {nodes [
   n1 = SetupCaptureEnviormentInPIE!() -> (then, PlayerPos:Vector=~) {is=title K2Node_FunctionEntry}
   > n2 = GetPIEWorlds(bIncludeDedicatedServer=false)
   n3 = "Get (a copy)"(Array=@n2)
   > ExecuteConsoleCommand(WorldContextObject=@n3, Command=cheat ui.hide)
   > ExecuteConsoleCommand(WorldContextObject=@n3, Command=DisableAllScreenMessages)
   > ExecuteConsoleCommand(WorldContextObject=@n3, Command=gc.TimeBetweenPurgingPendingKillObjects 240)
   n4 = BreakVector(InVec=@n1)
   > ExecuteConsoleCommand(
    WorldContextObject=@n3
    Command=Append!(
     A="BugItGo "
     B="To String (Float)"(InDouble=@n4)
     C=" "
     D="To String (Float)"(InDouble=@n4.Y)
     E=" "
     F:String="To String (Float)"(InDouble=@n4.Z)
     G:String=" "
     H:String=0
     I:String=" "
     J:String=-90
     K:String=" "
     L:String=0
    ) -> (ReturnValue)
   )
   > ExecuteConsoleCommand(Command=cheat SetTimeOfDayInHours 12)
   > ExecuteConsoleCommand(Command=cheat SetTimeOfDaySpeed 0)
   > Return Node!(execute, PIEWorld:Object Reference=@n3)
  ]}
 VersionCheck {nodes [
   n1 = VersionCheck!() -> (then, InObj:Object Reference) {is=title K2Node_FunctionEntry}
   > n2 = Cast To FortItemDefinition(Object=@n1)
   > n3 = EditorOnly_GetItemDefinitionFortReleaseVersion!(execute, ItemDefinition:Item Definition Base Object Reference=@n2) -> (then, ReturnValue:Fort Release Version Structure)
   > n4 = EditorOnly_GetVersionSeasonNumber!(execute, Version:"Fort Release Version Structure (by ref)"=@n3) -> (then, ReturnValue:int=0)
   > PrintString(InString="To String (Integer)"(InInt=@n4))
   > n5 = Branch(Condition="integer <= integer"(A=@n4, B=Get MaxVersionLimitation() -> (MaxVersionLimitation:int=0)))
   > n6 = Branch(Condition="integer >= integer"(A=@n4, B=Get MinVersionLimitation() -> (MinVersionLimitation:int=0)))
   > Return Node!(execute, VersionOK:bool=true)
   Return Node(execute=[@n2.CastFailed, @n6.else, @n5.else], VersionOK=false)
  ]}
 WriteCosmeticInfoToOriginalFolder {nodes [
   n1 = WriteCosmeticInfoToOriginalFolder!() -> (then, AssetData:Asset Data Structure) {is=title K2Node_FunctionEntry}
   > n2 = Sequence!(execute) -> (then_0, then_1, then_2, then_3)
   n3 = Append!(A=Get OriginalFolderPath(), B=Dependencies.txt) -> (ReturnValue)
   > WriteStringIntoFile(FilePath=@n3, Unique=false, Clear=false)
   n4 = GetItemFXResources(execute=@n2.then_1, ItemAssetData=@n1) {is=title K2Node_CallFunction}
   > n5 = Sequence!(execute) -> (then_0, then_1, then_2)
   > n6 = For Each Loop(Array:Array of Strings=@n4.FXList) -> (Array Element:String)
   > WriteStringIntoFile(WriteString=@n6, FilePath=@n3, Unique=true, Clear=false)
   n7 = For Each Loop(Exec=@n5.then_1, Array=@n4.MatList)
   > WriteStringIntoFile(WriteString=@n7, FilePath=@n3, Unique=true, Clear=false)
   n8 = For Each Loop(Exec=@n5.then_2, Array=@n4.TexList)
   > WriteStringIntoFile(WriteString=@n8, FilePath=@n3, Unique=true, Clear=false)
   Cast To NiagaraSystem(execute=@n2.then_2, Object=GetAsset(InAssetData=@n1))
   n9 = GetFullName(InAssetData=@n1)
   > WriteStringIntoFile(WriteString=@n9, FilePath=@n3, Unique=true, Clear=false)
   n10 = GetAsset(InAssetData=@n1)
   n11 = Cast To FortItemDefinition!(Object=@n10) -> (AsFort Item Definition, bSuccess:bool=~)
   n12 = EditorOnly_GetItemDefinitionFortReleaseVersion(execute=@n2.then_3, ItemDefinition=@n11)
   > n13 = EditorOnly_GetVersionSeasonNumber(Version=@n12) -> (ReturnValue=0)
   n14 = Get FlipbookResX() -> (FlipbookResX:int=0)
   n15 = Get FlipbookResY() -> (FlipbookResY:int=0)
   n16 = Get ItemName!(self:Item Definition Base Object Reference=@n11) -> (ItemName:Text)
   n17 = Get CaptureNSTemplate()
   > WriteStringIntoFile(
    WriteString=Append!(
     A=SelectString(A=null, B=@n9, bPickA=@n17)
     B=",&&,"
     C:String=GetObjectName(Object=@n10)
     D:String=",&&,"
     E:String="To String (Text)"(InText=@n16)
     F:String=",&&,"
     G:String="To String (Integer)"(InInt=@n13)
     H:String=",&&,"
     I:String="To String (Boolean)"(InBool=@n4)
     J:String=",&&,"
     K:String="To String (Integer)"(InInt=@n14)
     L:String=",&&,"
     M:String="To String (Integer)"(InInt=@n15)
    ) -> (ReturnValue)
    FilePath=Append!(A=Get OriginalFolderPath(), B=BasicInformation.txt) -> (ReturnValue)
    Unique=true
    Clear=true
   )
  ]}
]
variables [
 ArrangeFolderPath {type=string, default=D:/yi.xie.PC2.FXBrowserLib/depot/ArtTools/FXBrowser/Library/}
 BackpackList {type="array<AssetData>", default []}
 CaptureBase {type=Bp_CaptureBase_C, default=null}
 CaptureNSDataList {type="array<AssetData>", default []}
 CaptureNSList {type="array<NiagaraSystem>", default []}
 CaptureNSTemplate {type=boolean, default=true}
 CaptureType {type=CosmeticPicker_CosmeticTypes, default=Pickaxe}
 CheckEmote {type=boolean}
 Count {type=integer, default=1}
 CurCaptureNiagara {type=NiagaraSystem, default=null, class=~, dependencies=~, emitters=~, summary=~}
 DebugRefAsset {type=FortItemDefinition, default=null}
 DefaultStartIndex {type=integer}
 EmoteEndLogic {type=boolean}
 EmoteList {type="array<AssetData>", default []}
 FlipbookResX {type=integer, default=6}
 FlipbookResY {type=integer, default=6}
 GliderList {type="array<AssetData>", default []}
 MaxBound {type=BoxSphereBounds, default {sphereRadius=0, boxExtent=<0 0 0>, origin=<0 0 0>}}
 MaxVersionLimitation {type=integer, default=30}
 MinVersionLimitation {type=integer, default=10}
 OriginalFolderPath {type=string, default=D:&Original/}
 OverrideExist {type=boolean}
 PickaxeItemDef {type=Class, default=&AthenaPickaxeItemDefinition}
 PickaxeList {type="array<AssetData>", default []}
 PythonLinkFilePath {type=string, default=D:/RunningState.txt}
 ResetReusableCount {type=integer}
 ReusableNiagaraNS {type=NiagaraActor, default=null}
 ReusableNiagaraNSFunctionTime {type=number}
 RotateCharacter {type=boolean}
 SkinList {type="array<AssetData>", default []}
 SkipNoFX {type=boolean, default=true}
 SkipWarmupForDebug {type=boolean}
 SkydiveList {type="array<AssetData>", default []}
 StartReusableNiagaraNSBoundCheck {type=boolean}
 StartReusableNiagaraNSReset {type=boolean}
 TestID {type=FortItemDefinition, default=null}
]
widgets {
 info {rootWidgetClass=&CanvasPanel, widgetCount=9}
 widgets [
  CanvasPanel_37 {widgetClassPath=&CanvasPanel, widget=CanvasPanel_37}
  DetailsView_0 {
   widgetClassPath=&DetailsView
   parent=CanvasPanel_37
   widget=DetailsView_0
   properties {categoriesToShow [List]}
   slot {
    class=CanvasPanelSlot
    properties {layoutData {alignment=<0 0>, anchors {maximum=<0 1>, minimum=<0 0>}, offsets=<ltrb 12 12 360.06 11.0202>}}
   }
  }
  Start {
   widgetClassPath=&EditorUtilityButton
   parent=CanvasPanel_37
   widget=Start
   slot {class=CanvasPanelSlot, properties {layoutData {alignment=<0 0>, anchors {maximum=<0 0>, minimum=<0 0>}, offsets=<ltrb 388 72 200.0273 40>}}}
  }
  TextBlock_81 {
   widgetClassPath=&TextBlock
   parent=Start
   widget=TextBlock_81
   properties {
    text=StartCapturePickaxe
    font {
     bForceMonospaced=false
     bMaterialIsStencil=false
     fontMaterial=null
     fontObject=&Roboto
     letterSpacing=0
     monospacedWidth=1
     size=12
     skewAmount=0
     typefaceFontName=Bold
     outlineSettings {
      bApplyOutlineToDropShadows=false
      bMiteredCorners=false
      bSeparateFillAlpha=false
      outlineMaterial=null
      outlineSize=0
      outlineColor=<rgba 0 0 0 1>
     }
    }
   }
   slot {class=ButtonSlot}
  }
  GetAll {
   widgetClassPath=&EditorUtilityButton
   parent=CanvasPanel_37
   widget=GetAll
   slot {class=CanvasPanelSlot, properties {layoutData {alignment=<0 0>, anchors {maximum=<0 0>, minimum=<0 0>}, offsets=<ltrb 388 24 200.0273 40>}}}
  }
  TextBlock {
   widgetClassPath=&TextBlock
   parent=GetAll
   widget=TextBlock
   properties {
    text=GetAllCosmetic
    font {
     bForceMonospaced=false
     bMaterialIsStencil=false
     fontMaterial=null
     fontObject=&Roboto
     letterSpacing=0
     monospacedWidth=1
     size=12
     skewAmount=0
     typefaceFontName=Bold
     outlineSettings {
      bApplyOutlineToDropShadows=false
      bMiteredCorners=false
      bSeparateFillAlpha=false
      outlineMaterial=null
      outlineSize=0
      outlineColor=<rgba 0 0 0 1>
     }
    }
   }
   slot {class=ButtonSlot}
  }
  Start_1 {
   widgetClassPath=&EditorUtilityButton
   parent=CanvasPanel_37
   widget=Start_1
   slot {class=CanvasPanelSlot, properties {layoutData {alignment=<0 0>, anchors {maximum=<0 0>, minimum=<0 0>}, offsets=<ltrb 388 116 200.0273 40>}}}
  }
  TextBlock_1 {
   widgetClassPath=&TextBlock
   parent=Start_1
   widget=TextBlock_1
   properties {
    text=DoTest
    font {
     bForceMonospaced=false
     bMaterialIsStencil=false
     fontMaterial=null
     fontObject=&Roboto
     letterSpacing=0
     monospacedWidth=1
     size=12
     skewAmount=0
     typefaceFontName=Bold
     outlineSettings {
      bApplyOutlineToDropShadows=false
      bMiteredCorners=false
      bSeparateFillAlpha=false
      outlineMaterial=null
      outlineSize=0
      outlineColor=<rgba 0 0 0 1>
     }
    }
   }
   slot {class=ButtonSlot}
  }
  Image_0 {
   widgetClassPath=&Image
   bIsVariable=true
   parent=CanvasPanel_37
   widget=Image_0
   slot {class=CanvasPanelSlot, properties {layoutData {alignment=<0 0>, anchors {maximum=<0 0>, minimum=<0 0>}, offsets=<ltrb 408 204 100 40>}}}
  }
 ]
}
