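# M_DefaultFloor is a minimal Default Lit material in the FXBrowser MaintainTool folder, named as a default floor surface.
# Its graph holds a single Constant node left at its usual value, with no connection to a material output shown.
# Automatic setting of usage flags in the editor is turned off.
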Material /Game/NeverCook/FX/FXBrowser/MaintainTool/M_DefaultFloor
graph {properties {bAutomaticallySetUsageInEditor=false, usedShadingModels=Default Lit}, nodes [Constant()]}
